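# Niagara effect for the Galileo emote that draws a short-lived, light blue ribbon trail following an animated skeletal mesh.
# A single AnimTrail emitter spawns trail particles from a user-supplied skeletal mesh, initializing them between bone or socket indices with a set width, color and brief lifetime.
# Update modules advance the trail, and a float curve fades alpha to zero by mid-life for the ribbon renderer.
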NiagaraSystem /BRCosmetics/Athena/Effects/Particles/Emotes/NS_Galileo_Emote/NS_Galileo_Emote_2
dependencies {usedDataInterfaces [&NiagaraDataInterfaceCurve], usedDynamicInputs [], usedModules [&SystemState]}
emitters {FN_AnimTrail_V5 {
  emitterUpdateScript [
   &EmitterState()
   &FN_SpawnAnimTrail_v2(Spawn Interval=2)
  ]
  particleSpawnScript [
   /Game/Effects/Niagara/Modules/AnimTrail/FN_InitializeAnimTrail/FN_InitializeAnimTrailBoneIndecies(Second Bone or Socket Index=1)
   /Game/Effects/Niagara/Modules/AnimTrail/FN_InitializeAnimTrail/FN_InitializeAnimTrail(AnimTrailWidth=1, Color=<rgba 0.2133 0.54 0.808 1>, Lifetime=0.25)
  ]
  particleUpdateScript [
   &SolveForcesAndVelocity(Clamp Velocity=false, Limit Acceleration=false)
   &FN_UpdateAnimTrail_v2()
   &Color(
    Color=$Particles.Initial.Color
    Scale Alpha=&FloatFromCurve(FloatCurve=DataInterface() {propertyValues {Curve {keys [<at -0.0046 1.0113>,<at 0.5 0>]}}})
   )
  ]
  rendererClasses [&NiagaraRibbonRendererProperties]
  renderers [{rendererClass=&NiagaraRibbonRendererProperties}]
 }}
summary {
 systemName=NS_Galileo_Emote_2
 emitters [{emitterName=FN_AnimTrail_V5, rendererClasses [&NiagaraRibbonRendererProperties]}]
 userVariables [
  User.RibbonIDOffset {type=NiagaraInt32}
  User.SkeletalMesh {
   type=NiagaraDataInterfaceSkeletalMesh
   defaultValue {
    struct=NiagaraExt_VariableValue_DataInterface
    value {dataInterface=&self:NiagaraDataInterfaceSkeletalMesh_0}
   }
  }
 ]
}
