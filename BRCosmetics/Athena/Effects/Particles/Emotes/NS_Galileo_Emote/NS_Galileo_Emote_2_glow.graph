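# Niagara system for the Galileo emote's blue glow and electric lightning effect.
# A single persistent mesh with a light pops in through a scale-overshoot curve, surrounded by five fading glow sprites in an elongated box.
# Two emitters flicker short-lived lightning meshes above and below center, using spawn group as a lerp alpha to pick the side.
# Rounded arcs also mirror their Z scale, spawn progressively along X with emitter age, and start intensely overbright.
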NiagaraSystem /BRCosmetics/Athena/Effects/Particles/Emotes/NS_Galileo_Emote/NS_Galileo_Emote_2_glow
dependencies {
 usedDataInterfaces [&NiagaraDataInterfaceCurve, &NiagaraDataInterfaceVectorCurve, &NiagaraDataInterfaceColorCurve]
 usedDynamicInputs []
 usedModules [/Niagara/Modules/System/SystemLifeCycle, &SystemState]
}
emitters {
 CenterLightning {
  emitterUpdateScript [
   &EmitterState()
   &EmitterScalability(
    Module.bApplySpawnCountScale=false
    Module.bApplySpawnCountScaleByDistanceFraction=false
    Module.bUseMaxDistance=false
    Module.bUseMinDistance=false
   ) {enabled=false}
   &EmitterLifeCycle(CompleteOnInactive=true, MaxLoopCount=1, NextLoopDelay=1.25, NextLoopDuration=3.55) {enabled=false}
   &SpawnRate(SpawnRate=20, Use Spawn Probability=false) {enabled=false}
   &SpawnBurst_Instantaneous(Spawn Count=1, Use Spawn Probability=false)
  ]
  particleSpawnScript [
   /Niagara/Modules/Spawn/Initialization/InitializeParticle(
    Color=<rgba 0 0.3347 1 1>
    Lifetime=2.5
    Mesh Scale=<1.5 1.25 1.25>
    Module.Write Color=true
    Module.Write Lifetime=true
    Module.Write Mass=true
    Module.Write Material Random=true
    Module.Write Position=true
    Module.Write Scale=true
    Module.Write SpriteRotation=false
    Module.Write SpriteSize=false
    Position=&SimulationPosition()
    Sprite Size=<25 25>
   )
   /Niagara/Modules/Spawn/Orientation/V2/InitialMeshOrientation(Orientation Vector=<-1 0 0>, Use Orientation Vector=true, Use Rotation Vector=false)
   &SystemLocation(MaskSpawnGroup=false, Offset=<18 0 10>)
  ]
  particleUpdateScript [
   &ParticleState(Kill Particles When Lifetime Has Elapsed=false)
   &ScaleColor(Scale RGB=&MakeVectorFromLinearColorRGB(Color=<rgba 1 1 1 1>), ScaleA=true, ScaleRGB=true)
   &ScaleMeshSize(Scale Factor=&VectorFromCurve(VectorCurve=DataInterface() {propertyValues {
       XCurve {keys [<at 0 0>,<at 0.0383 2.9592>,<at 0.0982 1.0267>,<at 1 1>]}
       YCurve {keys [<at 0 0>,<at 0.0474 6.1037>,<at 0.1012 1.0484>,<at 1 1>]}
       ZCurve {keys [<at 0 0>,<at 0.0546 6.1098>,<at 0.1041 1.0267>,<at 1 1>]}
      }}))
  ]
  rendererClasses [&NiagaraLightRendererProperties, &NiagaraMeshRendererProperties]
  renderers [
   {rendererClass=&NiagaraLightRendererProperties, rendererIndex=0}
   {rendererClass=&NiagaraMeshRendererProperties, rendererIndex=1}
  ]
 }
 FN_Center_Sprite {
  emitterUpdateScript [
   &EmitterState()
   &EmitterScalability(
    Module.bApplySpawnCountScale=false
    Module.bApplySpawnCountScaleByDistanceFraction=false
    Module.bUseMaxDistance=false
    Module.bUseMinDistance=false
   ) {enabled=false}
   &EmitterLifeCycle(CompleteOnInactive=true, MaxLoopCount=1, NextLoopDuration=4.9958) {enabled=false}
   &SpawnRate(SpawnRate=4, Use Spawn Probability=false) {enabled=false}
   &SpawnBurst_Instantaneous(Spawn Count=5, Use Spawn Probability=false)
  ]
  particleSpawnScript [
   &InitializeParticle(
    Color=<rgba 0.2261 0.8157 1 0.35>
    Lifetime=4.75
    Module.Write Color=true
    Module.Write Lifetime=true
    Module.Write Mass=true
    Module.Write Material Random=true
    Module.Write Position=true
    Module.Write Scale=false
    Module.Write SpriteRotation=false
    Module.Write SpriteSize=true
    Position=&SimulationPosition()
    Sprite Rotation=&UniformRangedFloat(Minimum=0, Maximum=360)
    Sprite Size=&Vector2DFromFloat(Value=&UniformRangedFloat(Minimum=45, Maximum=55))
   )
   /NiagaraFN/Sprite/FN_RandomizeSpriteUV()
   &DynamicMaterialParameters(
    Index 0 Param 3=&UniformRangedFloat(Minimum=0, Maximum=1)
    Module.Param0WriteEnabled=false
    Module.Param1WriteEnabled=false
    Module.Param2WriteEnabled=true
    Module.Param3WriteEnabled=false
   )
   &BoxLocation(
    Box Size=<70 8 8>
    MaskSpawnGroup=false
    Module.Override Seed=false
    Randomness Mode:ENiagaraRandomnessMode=::NewEnumerator1
    Rotate Box=false
    Surface Only=false
   )
   /SetVariables(Particles.Position=&Add_Vector(A=Linked() {linkedVariable {name=Particles.Position, type=NiagaraPosition}}, B=<-6.5 0 8>)) {isSetParameters=true}
  ]
  particleUpdateScript [
   &ParticleState()
   &ScaleColor(
    Scale Alpha=&FloatFromCurve(
     FloatCurve=DataInterface() {propertyValues {Curve {keys [
         <at 0 0 RCIM_Cubic>
         {arriveTangent=1.3333, leaveTangent=1.3333, time=0.0944, value=1}
         {arriveTangent=-1.1007, interpMode=RCIM_Cubic, leaveTangent=-1.1007, time=0.9323, value=1}
         <at 1.0011 0.002 RCIM_Cubic>
        ]}}}
     Scale Curve=1.1
    )
    Scale RGB=&MakeVectorFromLinearColorRGB(Color=<rgba 0 0.4358 1 1>)
    ScaleA=true
    ScaleRGB=true
   )
   &ScaleSpriteSize(Scale Factor=&Vector2DFromFloat(Value=&FloatFromCurve(FloatCurve=DataInterface() {propertyValues {Curve {keys [<at -0.0046 0.7145>,<at 1.0011 0.9859>]}}})))
  ]
  rendererClasses [&NiagaraSpriteRendererProperties]
  renderers [{rendererClass=&NiagaraSpriteRendererProperties}]
 }
 TopBottomLightning {
  emitterUpdateScript [
   &EmitterState()
   &EmitterScalability(
    Module.bApplySpawnCountScale=false
    Module.bApplySpawnCountScaleByDistanceFraction=false
    Module.bUseMaxDistance=false
    Module.bUseMinDistance=false
   ) {enabled=false}
   &EmitterLifeCycle(CompleteOnInactive=true, MaxLoopCount=1, NextLoopDelay=0.9875, NextLoopDuration=3.55) {enabled=false}
   &SpawnRate(SpawnRate=&UniformRangedFloat(Minimum=0, Maximum=4), Use Spawn Probability=false)
   &SpawnBurst_Instantaneous(Spawn Time=1.65, Use Spawn Probability=false) {enabled=false}
   &SpawnRate(SpawnGroup=1, SpawnRate=&UniformRangedFloat(Minimum=0, Maximum=8), Use Spawn Probability=false) {module=SpawnRate001}
  ]
  particleSpawnScript [
   &InitializeParticle(
    Color=<rgba 0 0.4192 1 1>
    Lifetime=&UniformRangedFloat(Minimum=0.05, Maximum=0.15)
    Mesh Scale=&UniformRangedVector(Minimum=<0.0325 0.015 0.015>, Maximum=<0.055 0.05 0.05>)
    Module.Write Color=true
    Module.Write Lifetime=true
    Module.Write Mass=true
    Module.Write Material Random=true
    Module.Write Position=true
    Module.Write Scale=true
    Module.Write SpriteRotation=false
    Module.Write SpriteSize=false
    Position=&SimulationPosition()
    Sprite Size=<25 25>
   )
   &SystemLocation(MaskSpawnGroup=false, Offset=<8 0 10>)
   &InitialMeshOrientation(
    Rotation=&UniformRangedVector(Minimum=<360 0 0>, Maximum=<0 0 0>)
    Use Orientation Vector=true
    Use Rotation Vector=true
   )
   /SetVariables(Particles.Position=&Lerp_Vector(
     A=&Add_Vector(A=Linked() {linkedVariable {name=Particles.Position, type=NiagaraPosition}}, B=<14 0 6>)
     B=&Add_Vector(A=Linked() {linkedVariable {name=Particles.Position, type=NiagaraPosition}}, B=<14 0 -6>)
     Alpha=&MakeFloatFromInt(Int=$Emitter.SpawnGroup)
    )) {isSetParameters=true, module=SetVariables_#1}
  ]
  particleUpdateScript [
   &ParticleState()
   &ScaleColor(Scale RGB=&MakeVectorFromLinearColorRGB(Color=<rgba 1 1 1 1>), ScaleA=true, ScaleRGB=true)
  ]
  rendererClasses [&NiagaraMeshRendererProperties]
  renderers [{rendererClass=&NiagaraMeshRendererProperties, rendererIndex=0}]
 }
 TopBottomLightningRoundedArcs {
  emitterUpdateScript [
   &EmitterState()
   &EmitterScalability(
    Module.bApplySpawnCountScale=false
    Module.bApplySpawnCountScaleByDistanceFraction=false
    Module.bUseMaxDistance=false
    Module.bUseMinDistance=false
   ) {enabled=false}
   &EmitterLifeCycle(CompleteOnInactive=true, MaxLoopCount=1, NextLoopDelay=0.9875, NextLoopDuration=3.55) {enabled=false}
   &SpawnRate(SpawnRate=&UniformRangedFloat(Minimum=0, Maximum=15), Use Spawn Probability=false)
   &SpawnRate(SpawnGroup=1, SpawnRate=&UniformRangedFloat(Minimum=0, Maximum=15), Use Spawn Probability=false) {module=SpawnRate001}
   &SpawnBurst_Instantaneous(Spawn Time=1.65, Use Spawn Probability=false) {enabled=false}
  ]
  particleSpawnScript [
   &InitializeParticle(
    Color=<rgba 0 0.4192 1 1>
    Lifetime=&UniformRangedFloat(Minimum=0.05, Maximum=0.225)
    Mesh Scale=<0.055 0.1 0.1>
    Module.Write Color=true
    Module.Write Lifetime=true
    Module.Write Mass=true
    Module.Write Material Random=true
    Module.Write Position=true
    Module.Write Scale=true
    Module.Write SpriteRotation=false
    Module.Write SpriteSize=false
    Position=&SimulationPosition()
    Sprite Size=<25 25>
   )
   &InitialMeshOrientation(
    Orientation Vector=<0 1 0>
    Rotation=&UniformRangedVector(Minimum=<0 0.25 0>, Maximum=<0 -0.25 0>)
    Use Orientation Vector=true
    Use Rotation Vector=true
   )
   &SystemLocation(MaskSpawnGroup=false, Offset=<24 0 10>)
   /SetVariables(
    Particles.Position=&Lerp_Vector(
     A=&Add_Vector(
      A=Linked() {linkedVariable {name=Particles.Position, type=NiagaraPosition}}
      B=&MakeVector(
       X=&FloatFromCurve(FloatCurve=DataInterface() {propertyValues {Curve {keys [<at 0 -24>,<at 3.4 14>]}}}, CurveIndex=$Emitter.Age)
       Y=0
       Z=6
      )
     )
     B=&Add_Vector(
      A=Linked() {linkedVariable {name=Particles.Position, type=NiagaraPosition}}
      B=&MakeVector(
       X=&FloatFromCurve(FloatCurve=DataInterface() {propertyValues {Curve {keys [<at 0 -24>,<at 3.4 14>]}}}, CurveIndex=$Emitter.Age)
       Y=0
       Z=-6
      )
     )
     Alpha=&MakeFloatFromInt(Int=$Emitter.SpawnGroup)
    )
    Particles.Scale=&Lerp_Vector(
     A=&Multiply_Vector(A=$Particles.Scale, B=<1 1 1>)
     B=&Multiply_Vector(A=$Particles.Scale, B=<1 1 -1>)
     Alpha=&MakeFloatFromInt(Int=$Emitter.SpawnGroup)
    )
   ) {isSetParameters=true, module=SetVariables_#1}
   &DynamicMaterialParameters(
    Index 0 Param 1=&UniformRangedFloat(Minimum=0.225, Maximum=0.7)
    Index 0 Param 2=&UniformRangedFloat(Minimum=0, Maximum=1)
    Index 0 Param 3=100
    Index 0 Param 4=&UniformRangedFloat(Minimum=0, Maximum=10)
    Module.Param0WriteEnabled=true
    Module.Param1WriteEnabled=true
    Module.Param2WriteEnabled=true
    Module.Param3WriteEnabled=true
   )
  ]
  particleUpdateScript [
   &ParticleState()
   &ScaleColor(
    Scale Alpha=&FloatFromCurve(FloatCurve=DataInterface() {propertyValues {Curve {keys [<at 0 0>,<at 0.2491 1.0013>,<at 1.0011 0.002>]}}})
    Scale RGB=&MakeVectorFromLinearColorRGB(Color=&ScaleLinearColorByCurve(
      InitialValue=<rgba 1 1 1 1>
      ScaleCurve=DataInterface() {propertyValues {
        AlphaCurve {keys []}
        BlueCurve {keys [<at 0.0088 1000>,<at 0.5016 1>]}
        GreenCurve {keys [<at 0.0088 1000>,<at 0.5016 1>]}
        RedCurve {keys [<at 0.0088 1000>,<at 0.5016 1>]}
       }}
      CurveIndex=$Particles.NormalizedAge
     ))
    ScaleA=true
    ScaleRGB=true
   )
  ]
  rendererClasses [&NiagaraMeshRendererProperties]
  renderers [{rendererClass=&NiagaraMeshRendererProperties, rendererIndex=0}]
 }
}
summary {
 systemName=NS_Galileo_Emote_2_glow
 emitters [
  {emitterName=FN_Center_Sprite, rendererClasses [&NiagaraSpriteRendererProperties]}
  {emitterName=CenterLightning, rendererClasses [&NiagaraLightRendererProperties, &NiagaraMeshRendererProperties]}
  {emitterName=TopBottomLightning, rendererClasses [&NiagaraMeshRendererProperties]}
  {emitterName=TopBottomLightningRoundedArcs, rendererClasses [&NiagaraMeshRendererProperties]}
 ]
 userVariables []
}
