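# Niagara system for the Galileo emote (variant 3), producing a blue glow with a trail, sprites and a light.
# An anim-trail ribbon emitter spawns continuously through an offset ribbon initializer and fades over particle life, with user parameters exposed for skeletal mesh and ribbon ID offset.
# A one-shot burst scatters randomized sprites along a thin Y-axis box, fading and growing them.
# A single-particle emitter adds an offset blue point light.
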NiagaraSystem /BRCosmetics/Athena/Effects/Particles/Emotes/NS_Galileo_Emote/NS_Galileo_Emote_3
dependencies {usedDataInterfaces [&NiagaraDataInterfaceCurve], usedDynamicInputs [], usedModules [&SystemState]}
emitters {
 FN_AnimTrail_V2 {
  emitterUpdateScript [
   &EmitterScalability(
    Module.bApplySpawnCountScale=false
    Module.bApplySpawnCountScaleByDistanceFraction=false
    Module.bUseMaxDistance=false
    Module.bUseMinDistance=false
   ) {enabled=false}
   &EmitterLifeCycle(NextLoopDuration=0.8917) {enabled=false}
   &EmitterState(Life Cycle Mode=::NewEnumerator1, Loop Behavior=::NewEnumerator1, Loop Duration=2.5, UseLoopDelay=false)
   &SpawnRate(SpawnRate=20, Use Spawn Probability=false)
  ]
  particleSpawnScript [/Game/Effects/Niagara/Modules/Contrails/FN_InitializeRibbon/FN_InitializeRibbon(Color:LinearColor=<rgba 0.202 0.4085 1 0.225>, Lifetime=0.65, Offset Amount=-65, Ribbon Width=32)]
  particleUpdateScript [
   &SolveForcesAndVelocity(Clamp Velocity=false, Limit Acceleration=false)
   &FN_UpdateRibbon(Initial Color:LinearColor=$Particles.Initial.Color, Initial RibbonWidth:NiagaraFloat=$Particles.Initial.RibbonWidth)
   &Color(
    Color=$Particles.Initial.Color
    Scale Alpha=&FloatFromCurve(FloatCurve=DataInterface() {propertyValues {Curve {keys [{arriveTangent=-4.1501, interpMode=RCIM_Cubic, time=-0.0046, value=1.0113}, <at 0.5 0>]}}})
   )
  ]
  rendererClasses [&NiagaraRibbonRendererProperties]
  renderers [{rendererClass=&NiagaraRibbonRendererProperties}]
 }
 FN_Center_Sprite001 {
  emitterUpdateScript [
   &EmitterScalability(
    Module.bApplySpawnCountScale=false
    Module.bApplySpawnCountScaleByDistanceFraction=false
    Module.bUseMaxDistance=false
    Module.bUseMinDistance=false
   ) {enabled=false}
   &EmitterLifeCycle(CompleteOnInactive=true, NextLoopDuration=2) {enabled=false}
   &SpawnRate(SpawnRate=4, Use Spawn Probability=false) {enabled=false}
   &EmitterState()
   &SpawnBurst_Instantaneous(Spawn Count=8, Use Spawn Probability=false)
  ]
  particleSpawnScript [
   /Niagara/Modules/Spawn/Initialization/InitializeParticle(
    Color=<rgba 0.2261 0.8157 1 0.25>
    Lifetime=2.75
    Module.Write Color=true
    Module.Write Lifetime=true
    Module.Write Mass=true
    Module.Write Material Random=true
    Module.Write Position=true
    Module.Write Scale=false
    Module.Write SpriteRotation=false
    Module.Write SpriteSize=true
    Position=&SimulationPosition()
    Sprite Rotation=&UniformRangedFloat(Minimum=0, Maximum=360)
    Sprite Size=&Vector2DFromFloat(Value=&UniformRangedFloat(Minimum=45, Maximum=70))
   )
   &BoxLocation(
    Box Size=<1 150 1>
    MaskSpawnGroup=false
    Module.Override Seed=false
    Randomness Mode:ENiagaraRandomnessMode=::NewEnumerator1
    Rotate Box=false
    Surface Only=false
   )
   /NiagaraFN/Sprite/FN_RandomizeSpriteUV()
   &DynamicMaterialParameters(
    Index 0 Param 3=&UniformRangedFloat(Minimum=0, Maximum=1)
    Module.Param0WriteEnabled=false
    Module.Param1WriteEnabled=false
    Module.Param2WriteEnabled=true
    Module.Param3WriteEnabled=false
   )
   /SetVariables(Particles.Position=&Add_Vector(A=Linked() {linkedVariable {name=Particles.Position, type=NiagaraPosition}}, B=<0 0 0>)) {isSetParameters=true}
  ]
  particleUpdateScript [
   &ParticleState()
   &ScaleColor(
    Scale Alpha=&FloatFromCurve(FloatCurve=DataInterface() {propertyValues {Curve {keys [
         {arriveTangent=31.1685, interpMode=RCIM_Cubic, leaveTangent=31.1685, tangentMode=RCTM_User, time=0, value=0}
         {arriveTangent=1.3333, leaveTangent=1.3333, time=0.088, value=1}
         {arriveTangent=-1.093, interpMode=RCIM_Cubic, leaveTangent=-1.093, time=0.7044, value=1.0023}
         <at 1.0011 0.002 RCIM_Cubic>
        ]}}})
    Scale RGB=&MakeVectorFromLinearColorRGB(Color=<rgba 0 0.4358 1 1>)
    ScaleA=true
    ScaleRGB=true
   )
   &ScaleSpriteSize(Scale Factor=&Vector2DFromFloat(Value=&FloatFromCurve(FloatCurve=DataInterface() {propertyValues {Curve {keys [<at -0.0046 0.7145>,<at 1.0011 0.9859>]}}})))
  ]
  rendererClasses [&NiagaraSpriteRendererProperties]
  renderers [{rendererClass=&NiagaraSpriteRendererProperties}]
 }
 Light {
  emitterUpdateScript [
   &EmitterState()
   &SpawnBurst_Instantaneous(Spawn Count=1, Use Spawn Probability=false)
  ]
  particleSpawnScript [&InitializeParticle(
    Lifetime=2
    Module.Write Color=true
    Module.Write Lifetime=true
    Module.Write Mass=true
    Module.Write Material Random=true
    Module.Write Position=true
    Module.Write Scale=false
    Module.Write SpriteRotation=false
    Module.Write SpriteSize=true
    Sprite Size=<5 5>
   )]
  particleUpdateScript [
   &ParticleState()
   &FN_Light_v3(
    Color=<rgba 0 0.4192 1 1>
    Create Color Attribute=true
    Create Exponent Attribute=true
    Create Position Attribute=true
    Create Radius Attribute=true
    Enable Particle Light Enable=true
    Exponent=2
    Light Enabled=true
    Position:Vector3f=&Add_Vector(A=Linked() {linkedVariable {name=Particles.Position, type=NiagaraPosition}}, B=<0 -60 0>)
    Radius=100
    Scale Color=10
   )
  ]
  rendererClasses [&NiagaraLightRendererProperties]
  renderers [{rendererClass=&NiagaraLightRendererProperties, rendererIndex=0}]
 }
}
summary {
 systemName=NS_Galileo_Emote_3
 emitters [
  {emitterName=FN_Center_Sprite001, rendererClasses [&NiagaraSpriteRendererProperties]}
  {emitterName=FN_AnimTrail_V2, rendererClasses [&NiagaraRibbonRendererProperties]}
  {emitterName=Light, rendererClasses [&NiagaraLightRendererProperties]}
 ]
 userVariables [
  User.RibbonIDOffset {type=NiagaraInt32}
  User.SkeletalMesh {
   type=NiagaraDataInterfaceSkeletalMesh
   defaultValue {
    struct=NiagaraExt_VariableValue_DataInterface
    value {dataInterface=&self:NiagaraDataInterfaceSkeletalMesh_0}
   }
  }
 ]
}
